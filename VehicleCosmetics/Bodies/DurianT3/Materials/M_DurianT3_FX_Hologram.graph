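# Unlit, translucent, two-sided skeletal mesh material for a hologram effect on the DurianT3 vehicle body.
# Emissive color is the CustomColor parameter, cyan-blue by default, scaled by EmissivePower.
# Opacity adds time to the U texture coordinate, feeds a sine, and raises it to a high power, producing narrow bands that scroll along U.
# Both outputs pass through a SetMaterialAttributes node.
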Material /VehicleCosmetics/Bodies/DurianT3/Materials/M_DurianT3_FX_Hologram
graph {
 properties {
  bAutomaticallySetUsageInEditor=false
  bUseMaterialAttributes=true
  bUsedWithSkeletalMesh=true
  blendMode=BLEND_Translucent
  shadingModel=MSM_Unlit
  twoSided=true
  usedShadingModels=Unlit
 }
 nodes [
  n1 = Add(A=Multiply(A=Time()), B=ComponentMask(None=TextureCoordinate()) {r=true})
  SetMaterialAttributes!(
   MaterialAttributes
   Emissive Color=Multiply(
    A=VectorParameter() {parameterName=CustomColor, defaultValue=<rgba 0 0.6441 1 2>}.RGB
    B=ScalarParameter() {defaultValue=6, parameterName=EmissivePower}
   ) {constB=0.5}
   Opacity=Power(Base=Sine(None=@n1) {period=3}) {constExponent=12}
  ) {attributeSetTypes [#1, #2]}
 ]
}
