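# Physics asset for the BR variant of the DurianT3 vehicle cosmetic body mesh.
# It defines a single physics body on the root bone with no collision shapes listed.
# There are no constraints, so nothing in the asset articulates between bones.
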PhysicsAsset /VehicleCosmetics/Bodies/DurianT3/Meshes/Body_DurianT3_PhysicsBR
bodies {root []}
constraints []
